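# Sound cue for the Beast Girl Storm Claws BeastBaton jam track's heat transform percussion stinger.
# It picks one of two percussion waves at equal odds, then an enveloper ramps volume up over four seconds.
# It uses the player attacks and abilities sound class, references first-person emote music concurrency alongside a prevent-new override, and sends to the music audio-analysis submix.
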SoundCue /BeastGirlStormClawsGameplay/Assets/Audio/SoundCues/Music/BeastBaton_JamTrack_Heat_Transform_Stinger
root=Enveloper_0
nodes [
 {
  id=Enveloper_0
  children [Random_0]
  properties {
   pitchCurve {editorCurveData {keys [{time=0}]}}
   volumeCurve {editorCurveData {keys [{arriveTangent=0.2047, time=0, value=0.6}, <at 1.1113 0.7833>, <at 4 1>]}}
  }
 }
 {id=Random_0, children [WavePlayer_1,WavePlayer_3], properties {weights [1,1]}}
 {
  id=WavePlayer_1
  properties {soundWaveAssetPtr=/BeastGirlStormClawsGameplay/Assets/Audio/Wavs/Music/BeastBaton_JamTrack_Heat_Transform_Perc_01}
 }
 {
  id=WavePlayer_3
  properties {soundWaveAssetPtr=/BeastGirlStormClawsGameplay/Assets/Audio/Wavs/Music/BeastBaton_JamTrack_Heat_Transform_Perc_02}
 }
]
properties {
 maxDistance=2097152
 soundClassObject=/Game/Sounds/FortSoundClassesAndModes/Fort_06_Player_Attacks_Abilities
 volumeMultiplier=0.6562
 concurrencyOverrides {resolutionRule=PreventNew}
 concurrencySet [/Game/Sounds/Concurrency_Rules/Athena_EmoteMusic_1P_Conc]
 soundSubmixSends [{minSendLevel=0, soundSubmix=&Fortnite_AudioAnalysis_Music}]
}
