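# Third-person sound cue for the BeastBaton jam track Heat Transform music stinger, randomly playing one of two percussion waves.
# Nested branches on islocal and isenemy set volume: full for local or unset, reduced for others, with non-enemies quieter than enemies.
# It uses Beast Girl Storm music attenuation and emote-music concurrency, sits in the player attacks/abilities sound class, and has a music audio-analysis submix send.
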SoundCue /BeastGirlStormClawsGameplay/Assets/Audio/SoundCues/Music/BeastBaton_JamTrack_Heat_Transform_Stinger_3p
root=Branch_0
nodes [
 {id=Branch_0, children [Modulator_2,Branch_3,Modulator_2], properties {boolParameterName=islocal}}
 {id=Branch_3, children [Modulator_4,Modulator_3,Modulator_4], properties {boolParameterName=isenemy}}
 {id=Modulator_2, children [Random_0], properties {pitchMax=1, pitchMin=1, volumeMax=1, volumeMin=1}}
 {id=Modulator_3, children [Random_0], properties {pitchMax=1, pitchMin=1, volumeMax=0.35, volumeMin=0.35}}
 {id=Modulator_4, children [Random_0], properties {pitchMax=1, pitchMin=1, volumeMax=0.5, volumeMin=0.5}}
 {id=Random_0, children [WavePlayer_0,WavePlayer_1], properties {weights [1,1]}}
 {
  id=WavePlayer_0
  properties {soundWaveAssetPtr=/BeastGirlStormClawsGameplay/Assets/Audio/Wavs/Music/BeastBaton_JamTrack_Heat_Transform_Perc_01}
 }
 {
  id=WavePlayer_1
  properties {soundWaveAssetPtr=/BeastGirlStormClawsGameplay/Assets/Audio/Wavs/Music/BeastBaton_JamTrack_Heat_Transform_Perc_02}
 }
]
properties {
 attenuationSettings=/BeastGirlStormClawsGameplay/Assets/Audio/ATT_BeastGirlStorm_Music
 maxDistance=9908.75
 soundClassObject=/Game/Sounds/FortSoundClassesAndModes/Fort_06_Player_Attacks_Abilities
 volumeMultiplier=0.85
 concurrencyOverrides {resolutionRule=PreventNew}
 concurrencySet [/Game/Sounds/Concurrency_Rules/Athena_EmoteMusic_Conc]
 soundSubmixSends [{minSendLevel=0, soundSubmix=&Fortnite_AudioAnalysis_Music}]
}
